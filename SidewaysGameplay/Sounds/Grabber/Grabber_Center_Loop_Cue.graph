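# Sound cue for the looping center sound of the Grabber in SidewaysGameplay, built on a center pillar loop wave.
# A looping Wave Player feeds an Enveloper whose volume curve fades in quickly, then gradually eases to about half level over several seconds.
# The cue uses the Grabber ambient attenuation asset and the Fort_AGV sound class, with a reduced overall volume.
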SoundCue /SidewaysGameplay/Sounds/Grabber/Grabber_Center_Loop_Cue
root=Enveloper_0
nodes [
 {
  id=Enveloper_0
  children [WavePlayer_0]
  properties {
   pitchCurve {editorCurveData {keys [{time=0}]}}
   volumeCurve {editorCurveData {keys [<at 0 0.01>,<at 0.25 1>,<at 3.964 0.9921>,<at 6.9898 0.7056>,<at 11.397 0.5076>]}}
  }
 }
 {id=WavePlayer_0, properties {bLooping=true, soundWaveAssetPtr=&/Sideways_CenterPillar_Loop_01}}
]
properties {
 attenuationSettings=/SidewaysGameplay/Sounds/Grabber/Attenuations/Grabber_Amb_Att
 maxDistance=4000
 soundClassObject=/Game/Sounds/FortSoundClassesAndModes/Fort_AGV
 volumeMultiplier=0.65
}
