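# Sound cue for the Grabber core blast projectile, mixing two randomized layers at a low overall volume and slightly lowered pitch.
# One layer picks a distant core blast wave under the distant ranged-weapon fire class, with no attenuation asset shown.
# The other picks a distant Grabber spawn wave under the AGV class with close missile projectile attenuation.
# Both layers modulate pitch, and the cue references the Grabber spawn concurrency set.
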SoundCue /SidewaysGameplay/Sounds/Grabber/Grabber_Core_Blast_Prj_Cue
root=Mixer_0
nodes [
 {id=Attenuation_0, children [Modulator_0]}
 {
  id=Attenuation_3
  children [Modulator_6]
  properties {attenuationSettings=/Game/Sounds/Attenuation/Fort_Weapon_Projectile_Close_Missile_Athena_Att}
 }
 {id=Mixer_0, children [SoundClass_0,SoundClass_3], properties {inputVolume [1,1]}}
 {id=Modulator_0, children [Random_1], properties {pitchMax=1.1}}
 {id=Modulator_6, children [Random_4], properties {pitchMax=1.1}}
 {id=Random_1, children [WavePlayer_12, WavePlayer_13, WavePlayer_14], properties {weights [1,1,1]}}
 {
  id=Random_4
  children [WavePlayer_0, WavePlayer_7, WavePlayer_8, WavePlayer_9, WavePlayer_10, WavePlayer_11]
  properties {weights [1,1,1,1,1,1]}
 }
 {
  id=SoundClass_0
  children [Attenuation_0]
  properties {soundClassOverride=&Fort_WeaponRanged_Fire_Distant}
 }
 {id=SoundClass_3, children [Attenuation_3], properties {soundClassOverride=/Game/Sounds/FortSoundClassesAndModes/Fort_AGV}}
 {id=WavePlayer_0, properties {soundWaveAssetPtr=&/Sideways_Grabber_Spawn_Dist_05}}
 {id=WavePlayer_10, properties {soundWaveAssetPtr=&/Sideways_Grabber_Spawn_Dist_01}}
 {id=WavePlayer_11, properties {soundWaveAssetPtr=&/Sideways_Grabber_Spawn_Dist_06}}
 {id=WavePlayer_12, properties {soundWaveAssetPtr=&/Sideways_Core_Blast_Distant_01}}
 {id=WavePlayer_13, properties {soundWaveAssetPtr=&/Sideways_Core_Blast_Distant_02}}
 {id=WavePlayer_14, properties {soundWaveAssetPtr=&/Sideways_Core_Blast_Distant_03}}
 {id=WavePlayer_7, properties {soundWaveAssetPtr=&/Sideways_Grabber_Spawn_Dist_04}}
 {id=WavePlayer_8, properties {soundWaveAssetPtr=&/Sideways_Grabber_Spawn_Dist_03}}
 {id=WavePlayer_9, properties {soundWaveAssetPtr=&/Sideways_Grabber_Spawn_Dist_02}}
]
properties {
 maxDistance=2097152
 pitchMultiplier=0.9
 priority=0.5
 volumeMultiplier=0.15
 attenuationOverrides {bAttenuate=false}
 concurrencyOverrides {maxCount=1, resolutionRule=PreventNew}
 concurrencySet [&/Grabber_Spawn_Conc]
}
