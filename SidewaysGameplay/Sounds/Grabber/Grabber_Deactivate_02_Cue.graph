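# Sound cue for the Grabber deactivation sound in Sideways gameplay, second by name.
# A Modulator feeds from one Wave Player that reuses the Sideways destruction success wave, with cue-level pitch and volume lowered.
# Overridden attenuation sets its spatial falloff, and it uses a third-person ranged weapon fire sound class.
# Overridden concurrency allows one instance and blocks new ones, although a Grabber spawn concurrency asset is also referenced.
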SoundCue /SidewaysGameplay/Sounds/Grabber/Grabber_Deactivate_02_Cue
root=Modulator_0
nodes [
 {id=Modulator_0, children [WavePlayer_2]}
 {id=WavePlayer_2, properties {soundWaveAssetPtr=&/Sideways_DestructionSuccess_01}}
]
properties {
 bOverrideAttenuation=true
 bOverrideConcurrency=true
 maxDistance=7000
 pitchMultiplier=0.65
 priority=0.75
 soundClassObject=&Fort_WeaponRanged_Fire_3P_PSM
 volumeMultiplier=0.65
 attenuationOverrides {falloffDistance=4000, attenuationShapeExtents {x=3000}}
 concurrencyOverrides {maxCount=1, resolutionRule=PreventNew}
 concurrencySet [&/Grabber_Spawn_Conc]
}
