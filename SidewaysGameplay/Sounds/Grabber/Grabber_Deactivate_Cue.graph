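# Sound cue for the deactivation sound of the Grabber in SidewaysGameplay.
# A single wave player reuses a Sideways destruction-success wave and feeds an Enveloper with single-key pitch and volume curves.
# The cue lowers the volume, overrides attenuation for positional falloff, and routes to a third-person ranged weapon fire sound class.
# Its concurrency override allows one instance at a time and prevents new plays while one is active.
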SoundCue /SidewaysGameplay/Sounds/Grabber/Grabber_Deactivate_Cue
root=Enveloper_0
nodes [
 {
  id=Enveloper_0
  children [WavePlayer_1]
  properties {pitchCurve {editorCurveData {keys [{time=0}]}}, volumeCurve {editorCurveData {keys [<at 0 1>]}}}
 }
 {id=WavePlayer_1, properties {soundWaveAssetPtr=&/Sideways_DestructionSuccess_01_v2}}
]
properties {
 bOverrideAttenuation=true
 bOverrideConcurrency=true
 maxDistance=7000
 soundClassObject=&Fort_WeaponRanged_Fire_3P_PSM
 volumeMultiplier=0.65
 attenuationOverrides {falloffDistance=4000, attenuationShapeExtents {x=3000}}
 concurrencyOverrides {maxCount=1, resolutionRule=PreventNew}
 concurrencySet [null]
}
