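# Sound cue for the Sideways Grabber, apparently played when it fires its projectile, using tendril attack sounds.
# A mixer layers an equally weighted random pick of three tendril attack waves, at reduced volume, with a fixed tendril player-grab wave.
# The cue overrides attenuation for spatial falloff and overrides concurrency to allow one instance, blocking new plays while it sounds.
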SoundCue /SidewaysGameplay/Sounds/Grabber/Grabber_Fire_Prj_Cue
root=Mixer_0
nodes [
 {id=Mixer_0, children [Random_0,WavePlayer_1], properties {inputVolume [0.5,1]}}
 {id=Random_0, children [WavePlayer_0,WavePlayer_3,WavePlayer_4], properties {weights [1,1,1]}}
 {id=WavePlayer_0, properties {soundWaveAssetPtr=&/Sideways_Tendril_Attack_01_v2}}
 {id=WavePlayer_1, properties {soundWaveAssetPtr=&/Sideways_Tendril_Grab_Player_01}}
 {id=WavePlayer_3, properties {soundWaveAssetPtr=&/Sideways_Tendril_Attack_03_v2}}
 {id=WavePlayer_4, properties {soundWaveAssetPtr=&/Sideways_Tendril_Attack_04_v2}}
]
properties {
 bOverrideAttenuation=true
 bOverrideConcurrency=true
 maxDistance=6500
 soundClassObject=/Game/Sounds/FortSoundClassesAndModes/Fort_AGV
 volumeMultiplier=0.85
 attenuationOverrides {falloffDistance=5000, attenuationShapeExtents {x=1500}}
 concurrencyOverrides {maxCount=1, resolutionRule=PreventNew}
 concurrencySet [&/Grabber_Spawn_Conc]
}
