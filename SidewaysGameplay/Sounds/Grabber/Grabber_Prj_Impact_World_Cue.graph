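# Sound cue for the Grabber projectile's impact on the world, playing tendril impact sounds.
# A Random node picks one of six equally weighted impact waves and feeds an Enveloper whose volume curve holds briefly, then fades to silence in about half a second.
# The cue plays slightly lowered in pitch, overrides its attenuation, and uses a player world-impact concurrency set.
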SoundCue /SidewaysGameplay/Sounds/Grabber/Grabber_Prj_Impact_World_Cue
root=Enveloper_0
nodes [
 {
  id=Enveloper_0
  children [Random_0]
  properties {
   pitchCurve {editorCurveData {keys [{time=0}]}}
   volumeCurve {editorCurveData {keys [<at 0 1>,<at 0.177 0.9958>,<at 0.2198 0.6665>,<at 0.3178 0.341>,<at 0.5557 0>]}}
  }
 }
 {id=Random_0, children [WavePlayer_1,WavePlayer_2,WavePlayer_3,WavePlayer_4,WavePlayer_5,WavePlayer_6], properties {weights [1,1,1,1,1,1]}}
 {id=WavePlayer_1, properties {soundWaveAssetPtr=&/Sideways_Tendril_Impact_01_v2}}
 {id=WavePlayer_2, properties {soundWaveAssetPtr=&/Sideways_Tendril_Impact_02_v2}}
 {id=WavePlayer_3, properties {soundWaveAssetPtr=&/Sideways_Tendril_Impact_03_v2}}
 {id=WavePlayer_4, properties {soundWaveAssetPtr=&/Sideways_Tendril_Impact_04_v2}}
 {id=WavePlayer_5, properties {soundWaveAssetPtr=&/Sideways_Tendril_Impact_05_v2}}
 {id=WavePlayer_6, properties {soundWaveAssetPtr=&/Sideways_Tendril_Impact_06_v2}}
]
properties {
 bOverrideAttenuation=true
 maxDistance=2900
 pitchMultiplier=0.9
 priority=0.5
 soundClassObject=/Game/Sounds/FortSoundClassesAndModes/Fort_AGV
 attenuationOverrides {falloffDistance=2500}
 concurrencySet [&/Sideways_Player_World_Impacts_Conc]
}
