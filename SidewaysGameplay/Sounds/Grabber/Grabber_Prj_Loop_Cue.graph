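# Sound cue for the Grabber projectile loop in Sideways gameplay, going by its name.
# A single wave player feeds a tendril loop wave straight to the output, with no mixing, randomization or modulation nodes.
# It borrows the shared launcher in-air projectile attenuation and uses the Fort_AGV sound class.
# A concurrency override allows only one instance, resolved by stopping the farthest, then preventing new ones.
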SoundCue /SidewaysGameplay/Sounds/Grabber/Grabber_Prj_Loop_Cue
root=WavePlayer_2
nodes [{id=WavePlayer_2, properties {soundWaveAssetPtr=&/Sideways_Tendril_Loop_01_v2}}]
properties {
 attenuationSettings=/Game/Sounds/Attenuation/Fort_Weapon_Launcher_Proj_InAir_Default_Att
 bOverrideConcurrency=true
 maxDistance=12750
 soundClassObject=/Game/Sounds/FortSoundClassesAndModes/Fort_AGV
 concurrencyOverrides {maxCount=1, resolutionRule=StopFarthestThenPreventNew}
}
