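# Sound cue for a close projectile whip-by of the Grabber, judging by its name, routed to the Fort_AGV sound class.
# A randomized delay of roughly half a second feeds a single wave player that reuses a far shotgun bullet-whip sample, pitched up.
# Overridden attenuation turns off spatialization and reverb send and limits the sound to a short falloff range.
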SoundCue /SidewaysGameplay/Sounds/Grabber/Grabber_Prj_Whip_Close_Cue
root=Delay_0
nodes [
 {id=Delay_0, children [WavePlayer_1], properties {delayMax=0.65, delayMin=0.5}}
 {id=WavePlayer_1, properties {soundWaveAssetPtr=/Game/Athena/Sounds/Weapons/BulletWhips/Bullet_Whip_Shotgun_Far_02}}
]
properties {
 bOverrideAttenuation=true
 maxDistance=1500
 pitchMultiplier=1.25
 soundClassObject=/Game/Sounds/FortSoundClassesAndModes/Fort_AGV
 attenuationOverrides {
  bEnableReverbSend=false
  bSpatialize=false
  bUseComplexCollisionForOcclusion=true
  falloffDistance=1500
  occlusionVolumeAttenuation=0
  stereoSpread=0
  attenuationShapeExtents {x=0}
 }
}
