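# Sound cue for the distant bullet-whip sound of the Grabber projectile in SidewaysGameplay.
# A single wave player plays an Athena shotgun far bullet-whip sample, routed to the ranged weapon projectile effects sound class.
# The cue overrides attenuation inline with a short falloff, no reverb send and complex-collision occlusion.
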SoundCue /SidewaysGameplay/Sounds/Grabber/Grabber_Prj_Whip_Distant_Cue
root=WavePlayer_1
nodes [{id=WavePlayer_1, properties {soundWaveAssetPtr=/Game/Athena/Sounds/Weapons/BulletWhips/Bullet_Whip_Shotgun_Far_02}}]
properties {
 bOverrideAttenuation=true
 maxDistance=1500
 soundClassObject=&Fort_WeaponRanged_Projectile_Effects
 attenuationOverrides {
  bEnableReverbSend=false
  bUseComplexCollisionForOcclusion=true
  falloffDistance=1500
  occlusionVolumeAttenuation=0
  stereoSpread=0
  attenuationShapeExtents {x=0}
 }
}
